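# Physics asset for the Scribble backpack cosmetic mesh, defining secondary motion for its dangling parts.
# Small capsule bodies cover the pack, bag, ears, grenades with their rings, and two zipper pulls.
# The bag is tightly constrained to the pack, with ears, grenades and zips constrained to the bag and each ring chained to its grenade.
# Per-part angular limits shape the motion, letting rings and zips swing mainly on one axis while grenades mostly twist.
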PhysicsAsset /BRCosmetics/Accessories/FORT_Backpacks/Backpack_F_MED_Scribble/Meshes/F_MED_Scribble_Pack_Physics
bodies {
 bag [{shapeType=Capsule, length=2, radius=2, shapeName=bag_capsule}]
 dyn_ear_l [{shapeType=Capsule, length=2, radius=2, shapeName=dyn_ear_l_capsule, center=<4.3036 0 0>, rotation=<rot 90 0 0>}]
 dyn_ear_r [{shapeType=Capsule, length=2, radius=2, shapeName=dyn_ear_r_capsule, center=<4.3158 0 0>, rotation=<rot 90 0 0>}]
 dyn_grenade_l [{shapeType=Capsule, length=2, radius=2, shapeName=dyn_grenade_l_capsule, center=<3 0 0>, rotation=<rot 90 0 0>}]
 dyn_grenade_r [{shapeType=Capsule, length=2, radius=2, shapeName=dyn_grenade_r_capsule, center=<3 0 0>, rotation=<rot 90 0 0>}]
 dyn_ring_l [{shapeType=Capsule, length=1, radius=1, shapeName=dyn_ring_l_capsule, center=<2 0 0>, rotation=<rot 90 0 0>}]
 dyn_ring_r [{shapeType=Capsule, length=1, radius=1, shapeName=dyn_ring_r_capsule, center=<2 0 0>, rotation=<rot 90 0 0>}]
 dyn_zip_01 [{shapeType=Capsule, length=1, radius=1, shapeName=dyn_zip_01_capsule, center=<1.246 0 0>, rotation=<rot 90 0 0>}]
 dyn_zip_02 [{shapeType=Capsule, length=1, radius=1, shapeName=dyn_zip_02_capsule, center=<0.4422 0 0>, rotation=<rot 90 0 0>}]
 pack [{shapeType=Capsule, length=0.505, radius=0.505, shapeName=pack_capsule, rotation=<rot 0 -90 0>}]
}
constraints [
 {bone1Name=bag, bone2Name=pack, swing1LimitDegrees=5, swing2LimitDegrees=5, twistLimitDegrees=5}
 {bone1Name=dyn_ear_l, bone2Name=bag, swing1LimitDegrees=15, swing2LimitDegrees=15, twistLimitDegrees=1}
 {bone1Name=dyn_ear_r, bone2Name=bag, swing1LimitDegrees=15, swing2LimitDegrees=15, twistLimitDegrees=1}
 {bone1Name=dyn_grenade_l, bone2Name=bag, swing1LimitDegrees=10, swing2LimitDegrees=10, twistLimitDegrees=30}
 {bone1Name=dyn_ring_l, bone2Name=dyn_grenade_l, swing1LimitDegrees=40, swing2LimitDegrees=1, twistLimitDegrees=1}
 {bone1Name=dyn_grenade_r, bone2Name=bag, swing1LimitDegrees=10, swing2LimitDegrees=10, twistLimitDegrees=30}
 {bone1Name=dyn_ring_r, bone2Name=dyn_grenade_r, swing1LimitDegrees=40, swing2LimitDegrees=1, twistLimitDegrees=1}
 {bone1Name=dyn_zip_01, bone2Name=bag, swing1LimitDegrees=1, swing2LimitDegrees=45, twistLimitDegrees=1}
 {bone1Name=dyn_zip_02, bone2Name=bag, swing1LimitDegrees=1, swing2LimitDegrees=45, twistLimitDegrees=1}
]
